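# Parent base material for blockout plastic surfaces in the VehicleCosmetics content.
# It exposes Difuse color, Metalness and Roughness parameters, defaulting to dark grey, fully metallic and fairly smooth, with no textures.
# Usage is limited to explicit flags for skeletal meshes and mesh deformers, with automatic usage setting in the editor turned off.
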Material /VehicleCosmetics/Materials/BlockOut_Materials/Parents/MM_BlockoutPlastic_Base
graph {
 properties {
  bAutomaticallySetUsageInEditor=false
  bUsedWithMeshDeformer=true
  bUsedWithSkeletalMesh=true
  shadingModel=MSM_FromMaterialExpression
  usedShadingModels=Default Lit
 }
 nodes [
  ScalarParameter() {defaultValue=1, parameterName=Metalness}
  ScalarParameter() {defaultValue=0.4, parameterName=Roughness}
  VectorParameter() {parameterName=Difuse, defaultValue=<rgba 0.151 0.151 0.151 1>}
 ]
}
